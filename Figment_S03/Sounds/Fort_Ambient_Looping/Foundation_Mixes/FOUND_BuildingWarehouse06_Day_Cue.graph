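# Looping daytime interior ambience cue for a warehouse building foundation, playing an industrial large room tone stereo loop at reduced volume.
# A looping wave player feeds a modulator that slightly varies pitch and volume.
# It uses a square 3x3 map attenuation and an in-building foundation ambience sound class.
# A concurrency override prevents new instances once the limit is reached.
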SoundCue /Figment_S03/Sounds/Fort_Ambient_Looping/Foundation_Mixes/FOUND_BuildingWarehouse06_Day_Cue
root=Modulator_12
nodes [
 {id=Modulator_12, children [WavePlayer_79], properties {pitchMax=1.1, pitchMin=0.9, volumeMax=1.1, volumeMin=0.9}}
 {
  id=WavePlayer_79
  properties {bLooping=true, soundWaveAssetPtr=/Figment_S03/Sounds/Trans_AmbientLoops/RoomTones/IndustrialLarge006_Loop_Stereo}
 }
]
properties {
 attenuationSettings=/Game/Sounds/Attenuation/Fort_Map_3x3_Square
 bOverrideConcurrency=true
 maxDistance=1780
 soundClassObject=/Figment_S03/Sounds/FortSoundClassesAndModes/Foundations_Ambience_InBuilding_PSM
 volumeMultiplier=0.3
 concurrencyOverrides {resolutionRule=PreventNew}
}
