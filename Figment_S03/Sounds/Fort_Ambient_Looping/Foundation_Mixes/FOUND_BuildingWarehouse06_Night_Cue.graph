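# Looping night ambience cue for the BuildingWarehouse06 foundation, using a square map attenuation and the in-building ambience sound class.
# A quality-level node selects between two variants, leaving its last two slots empty.
# The first mixes industrial and mine room tones, with opposed 30-second volume envelopes that slowly crossfade them.
# The second plays only the industrial loop with drifting volume and pitch envelopes; modulators randomize pitch and volume.
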SoundCue /Figment_S03/Sounds/Fort_Ambient_Looping/Foundation_Mixes/FOUND_BuildingWarehouse06_Night_Cue
root=QualityLevel_0
nodes [
 {
  id=Enveloper_0
  children [WavePlayer_2]
  properties {
   bLoop=true
   bLoopIndefinitely=true
   loopEnd=30
   pitchCurve {editorCurveData {keys [
      {interpMode=RCIM_Cubic, time=0}
      {arriveTangent=0.0002, interpMode=RCIM_Cubic, leaveTangent=0.0002, tangentMode=RCTM_User, time=7.9875, value=1.0314}
      {arriveTangent=0, interpMode=RCIM_Cubic, leaveTangent=0, tangentMode=RCTM_User, time=20.0737, value=0.9355}
      {interpMode=RCIM_Cubic, time=30}
     ]}}
   volumeCurve {editorCurveData {keys [<at 0 1 RCIM_Cubic>,<at 15.6539 0.8429 RCIM_Cubic>,<at 30 1 RCIM_Cubic>]}}
  }
 }
 {
  id=Enveloper_6
  children [WavePlayer_100]
  properties {
   bLoop=true
   bLoopIndefinitely=true
   loopEnd=30
   pitchCurve {editorCurveData {keys [{time=0}]}}
   volumeCurve {editorCurveData {keys [<at 0 1 RCIM_Cubic>,<at 15 0.3 RCIM_Cubic>,<at 30 1 RCIM_Cubic>]}}
  }
 }
 {
  id=Enveloper_7
  children [Modulator_0]
  properties {
   bLoop=true
   bLoopIndefinitely=true
   loopEnd=30
   pitchCurve {editorCurveData {keys [{time=0}]}}
   volumeCurve {editorCurveData {keys [<at 0 0.3 RCIM_Cubic>,<at 15 1 RCIM_Cubic>,<at 30 0.3 RCIM_Cubic>]}}
  }
 }
 {id=Mixer_1, children [Enveloper_7,Enveloper_6], properties {inputVolume [0.5,1]}}
 {id=Modulator_0, children [WavePlayer_1], properties {pitchMax=0.8, pitchMin=0.5}}
 {id=Modulator_1, children [Enveloper_0], properties {pitchMax=0.8, pitchMin=0.5, volumeMax=1.1, volumeMin=0.9}}
 {id=Modulator_12, children [Mixer_1], properties {pitchMax=1.2, pitchMin=0.8, volumeMax=1.1, volumeMin=0.9}}
 {id=QualityLevel_0, children [Modulator_12,Modulator_1,null,null]}
 {
  id=WavePlayer_1
  properties {bLooping=true, soundWaveAssetPtr=/Figment_S03/Sounds/Trans_AmbientLoops/RoomTones/IndustrialLarge006_Loop_Stereo}
 }
 {id=WavePlayer_100, properties {bLooping=true, soundWaveAssetPtr=/Game/Sounds/Trans_AmbientLoops/RoomTones/MineA007_Loop_Stereo}}
 {
  id=WavePlayer_2
  properties {bLooping=true, soundWaveAssetPtr=&IndustrialLarge006_Loop_Stereo}
 }
]
properties {
 attenuationSettings=/Game/Sounds/Attenuation/Fort_Map_3x3_Square
 bOverrideConcurrency=true
 maxDistance=1780
 soundClassObject=/Figment_S03/Sounds/FortSoundClassesAndModes/Foundations_Ambience_InBuilding_PSM
 volumeMultiplier=0.5
 concurrencyOverrides {resolutionRule=PreventNew}
}
